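; Building actor Blueprint for an interior simple-brick wall piece with a side doorway, judging by its name and path.
; It carries no logic of its own, and its construction script is empty.
; The creative-mode bounds, origin and relative-transform queries return false with zero or default values.
; The grid axis guide attachment query returns two true flags and a zero vector.
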(fn GetCreativeActorBounds
  (return false "0, 0, 0"))

(fn GetCreativeActorOrigin
  (return false "0, 0, 0"))

(fn GetCreativeActorRelativeTransform
  (return false 0))

(fn GetGridAxisGuideAttachmentPoint
  (return true true "0, 0, 0"))

(fn ConstructionScript)
