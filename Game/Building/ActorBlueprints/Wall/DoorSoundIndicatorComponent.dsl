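; Sound indicator component for doors on building walls, deciding which players see a sound cue when the door makes noise.
; It shows indicators for Friendly, Neutral and Hostile affiliations.
; When the owner is a BuildingWall, it shows the indicator only to players whose pawn's player state differs from the requesting player state, so the requester is excluded.
; Owners that are not walls never show it.
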(fn GetAffiliationsToShowFor
  (return (Utilities|Array|MakeArray "Friendly" "Neutral" "Hostile")))

(fn ShouldShowSoundIndicator
  (params (PlayerController :type Fort%20Player%20Controller%20Object%20Reference))
  (Utilities|Casting|CastToBuildingWall (Components|GetOwner)
    (:then
      (return (!= (Variables|SoundIndicator|GetRequestingPlayerState) (Variables|/Script/Engine.Pawn|GetPlayerState (Game|Player|GetPlayerPawn PlayerController)))))
    (:CastFailed
      (return false))))
