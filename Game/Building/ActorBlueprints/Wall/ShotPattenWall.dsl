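; Shot pattern practice wall that records where shots land and reports distance, accuracy, shot count, time to kill and shots to kill.
; Cosmetic damage events spawn distance-scaled hit markers and rebuild text readouts each shot.
; Construction script caches body-part locations and tolerances; vector comparisons detect hits on a simulated 200-health player, doubling head damage.
; Shots near the reset button clear markers and reset health, timer and counters.
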(event EventBeginPlay
  (|CallParentFunction) ; This node is disabled and will not be called.\nDrag off pins to build functionality.
)

(event Collision|EventActorBeginOverlap
  (params (OtherActor :type Actor%20Object%20Reference))
  (|CallParentFunction OtherActor) ; This node is disabled and will not be called.\nDrag off pins to build functionality.
)

(event EventTick
  (params (DeltaSeconds :type Float%20%28single-precision%29))
  (|CallParentFunction DeltaSeconds) ; This node is disabled and will not be called.\nDrag off pins to build functionality.
)

(event EventOnDamagePlayEffects
  (params (Damage :type Float%20%28single-precision%29) (DamageTags :type Gameplay%20Tag%20Container%20Structure%20%28by%20ref%29) (Momentum :type Vector :default "0, 0, 0") (HitInfo :type Hit%20Result%20Structure%20%28by%20ref%29) (InstigatedBy :type Fort%20Pawn%20Object%20Reference) (DamageCauser :type Actor%20Object%20Reference) (EffectContext :type Gameplay%20Effect%20Context%20Handle%20Structure))
  (bind _returnvalue (Math|Float|Truncate Damage))
  (bind _distancefromtarget (Variables|Default|GetDistanceFromTarget))
  (bind _returnvalue_2 (- _distancefromtarget 5e+02))
  (bind _returnvalue_3 (/ _returnvalue_2 6e+03))
  (bind (_bblockinghit _binitialoverlap _time _distance _location) (Collision|BreakHitResult HitInfo))
  (Networking|CanExecuteCosmeticEvents
    (:True
      ;[1 1 1] Do initial cleanup
      (CallFunction|ClearTextArray (Variables|Default|GetDistanceText))
      (CallFunction|ClearTextArray (Variables|Default|GetAccuracyText))
      (CallFunction|ClearTextArray (Variables|Default|GetTotalShotsText))
      (CallFunction|ClearTextArray (Variables|Default|GetTimeToKillTimerText))
      (CallFunction|ClearTextArray (Variables|Default|GetShotToKillText))
      ;[1 1 1] Did player hit reset button?
      (if (Math|Vector|NotEqual(Vector) _location (Transformation|GetWorldLocation (Variables|Default|GetResetButtonCollisionBox)) 32.0)
        ;[1 1 1] Start Time to kill timer
        (if (Variables|Default|GetshouldTimerStartAtZero)
          (Variables|Default|SetFirstShotTimer (Utilities|Time|GetGameTimeinSeconds))
          (Variables|Default|SetshouldTimerStartAtZero false)
          (else
            _))
        (Variables|Default|SetTotalShots (+ 1 (Variables|Default|GetTotalShots)))
        ;[1 1 1] measure player distance from wall and scale marker size based on that distance
        (bind _distance_2 (Class|MissionBlueprintFunctionLibrary|DistanceBetweenTwoVectors (Transformation|GetActorLocation (Game|GetPlayerPawn 0)) _location))
        (Variables|Default|SetDistanceFromTarget _distance_2)
        (Math|Float|CompareFloat _returnvalue_3 0.25
          (:>
            (Variables|Default|SetScaleValue _returnvalue_3))
          (:==
            (Variables|Default|SetScaleValue 0.25))
          (:<
            _))
        (bind _returnvalue_4 (Game|SpawnActorfromClass "/Game/Building/ActorBlueprints/Wall/ShotPatternWall_HitMarker.ShotPatternWall_HitMarker_C" (Math|Transform|MakeTransform _location "0, 0, 0" (Math|Vector|MakeVector (Variables|Default|GetScaleValue) (Variables|Default|GetScaleValue) (Variables|Default|GetScaleValue))) "Undefined" "OverrideRootScale" 0 0))
        (Utilities|Array|Add (Variables|Default|GetHitMarkerObjects) _returnvalue_4)
        (Variables|Default|SetDistanceDisplay (Math|Float|Truncate (/ _distance_2 1e+02)))
        (bind (_array_element _array_index) (Utilities|Array|ForEachLoopwithBreak (Variables|Default|GetPlayerCollisionArray))
          (:LoopBody
            ;[1 1 1] Check each section of the player hitbox, if we hit something, assign Player Hit and break
            (if (not (Math|Vector|NotEqual(Vector) _array_element _location (Utilities|Array|Get(acopy) (Variables|Default|GetToleranceLevelsForPCA) _array_index)))
              (Variables|Default|SetPlayerHit true)
              ;;[1 0.6012 0.08416] Check if headshot to get accurate damage calculations (assumes 2x HS multiplier)
              (if (== _array_index 0)
                (Variables|Default|SetDamageDealtOnShot (* _returnvalue 2)))
              (Variables|Default|SetDamageDealtOnShot _returnvalue)))
          (:Completed
            _))
        ;[1 1 1] Initialize Variables dependent on Player Hit
        (if (Variables|Default|GetPlayerHit)
          (if (not (Variables|Default|GetIsPlayerDead))
            (bind _output_get (Variables|Default|SetPlayerHealth (- (Variables|Default|GetPlayerHealth) (Variables|Default|GetDamageDealtOnShot))))
            (if (<= _output_get 0)
              (Variables|Default|SetIsPlayerDead true)
              (Variables|Default|SetShotsToKill (Variables|Default|GetTotalShots))
              (Variables|Default|SetTimeToKill (- (Utilities|Time|GetGameTimeinSeconds) (Variables|Default|GetFirstShotTimer)))
              (else
                _)))
          (Variables|Default|SetShotsHitPlayer (+ (Variables|Default|GetShotsHitPlayer) 1))
          (else
            _))
        (Variables|Default|SetAccuracyToDisplay (* 1e+02 (/ (Math|Conversions|ToFloat(Integer) (Variables|Default|GetShotsHitPlayer)) (Math|Conversions|ToFloat(Integer) (Variables|Default|GetTotalShots)))))
        ;[1 0.1768 0.8859] Setup text components
        ;;[1 1 1] Distance Display text
        (bind _returnvalue_5 (AddComponent|NODE_AddTextRenderComponent-3 :RelativeTransform (Math|Transform|MakeTransform "0.000000,50.000000,380.000000" "0,90.000000,0.000000")))
        (Rendering|Components|TextRender|SetText _returnvalue_5 (Utilities|Text|ToText(String) (Utilities|String|Append "(" (Utilities|String|ToString(Integer) (Variables|Default|GetDistanceDisplay)) "m)")))
        (Utilities|Array|Add (Variables|Default|GetDistanceText) _returnvalue_5)
        ;;[1 1 1] Accuracy Text
        (bind _returnvalue_6 (AddComponent|NODE_AddTextRenderComponent-2 :RelativeTransform (Math|Transform|MakeTransform "203.000000,20.000000,290.000000" "0,90.000000,0.000000" "0.100000,0.100000,0.100000")))
        (Rendering|Components|TextRender|SetText _returnvalue_6 (Utilities|Text|ToText(String) (Utilities|String|Append "Accuracy: " (Utilities|String|Append (Utilities|Array|Get(acopy) (Utilities|String|ParseIntoArray (Utilities|String|ToString(Float) (Variables|Default|GetAccuracyToDisplay)) ".")) (Utilities|String|Append "." (Utilities|String|GetSubstring (Utilities|Array|Get(acopy) (Utilities|String|ParseIntoArray (Utilities|String|ToString(Float) (Variables|Default|GetAccuracyToDisplay)) ".") 1) 0 2))) "%")))
        (Utilities|Array|Add (Variables|Default|GetAccuracyText) _returnvalue_6)
        ;;[1 1 1] Total Shots taken
        (bind _returnvalue_7 (AddComponent|NODE_AddTextRenderComponent-6 :RelativeTransform (Math|Transform|MakeTransform "209.000000,20.000000,300.000000" "0,90.000000,0.000000" "0.100000,0.100000,0.100000")))
        (Rendering|Components|TextRender|SetText _returnvalue_7 (Utilities|Text|ToText(String) (Utilities|String|Append "Total Shots: " (Utilities|String|ToString(Integer) (Variables|Default|GetTotalShots)))))
        (Utilities|Array|Add (Variables|Default|GetTotalShotsText) _returnvalue_7)
        ;;[1 1 1] Time To Kill (200 health target)
        (if (Variables|Default|GetIsPlayerDead)
          (bind _returnvalue_8 (AddComponent|NODE_AddTextRenderComponent-7 :RelativeTransform (Math|Transform|MakeTransform "170.000000,20.000000,270.000000" "0,90.000000,0.000000" "0.100000,0.100000,0.100000")))
          (Rendering|Components|TextRender|SetText _returnvalue_8 (Utilities|Text|ToText(String) (Utilities|String|Append "Time to Kill Player: " (Utilities|String|Append (Utilities|Array|Get(acopy) (Utilities|String|ParseIntoArray (Utilities|String|ToString(Float) (Variables|Default|GetTimeToKill)) ".")) (Utilities|String|Append "." (Utilities|String|GetSubstring (Utilities|Array|Get(acopy) (Utilities|String|ParseIntoArray (Utilities|String|ToString(Float) (Variables|Default|GetTimeToKill)) ".") 1) 0 2))) " seconds")))
          (Utilities|Array|Add (Variables|Default|GetTimeToKillTimerText) _returnvalue_8)
          (else
            _))
        ;;[1 1 1] Shots To Kill (200 health)
        (if (Variables|Default|GetIsPlayerDead)
          (bind _returnvalue_9 (AddComponent|NODE_AddTextRenderComponent-8 :RelativeTransform (Math|Transform|MakeTransform "208.000000,20.000000,280.000000" "0,90.000000,0.000000" "0.100000,0.100000,0.100000")))
          (Rendering|Components|TextRender|SetText _returnvalue_9 (Utilities|Text|ToText(String) (Utilities|String|Append "Shots to kill: " (Utilities|String|ToString(Integer) (Variables|Default|GetShotsToKill)))))
          (Utilities|Array|Add (Variables|Default|GetShotToKillText) _returnvalue_9)
          (else
            _))
        ;[1 1 1] Reset variables that need to be reset prior to another shot
        (Variables|Default|SetPlayerHit false)
        (Variables|Default|SetDamageDealtOnShot 0)
        (else
          ;[1 0.0625 0.0625] Reset
          (Development|PrintString "ResetButton Hit")
          (for _array_element_2 (Variables|Default|GetHitMarkerObjects)
            (Actor|DestroyActor _array_element_2))
          (Utilities|Array|Clear (Variables|Default|GetHitMarkerObjects))
          (Variables|Default|SetTotalDamageDealt 0)
          (Variables|Default|SetTotalShots 0)
          (Variables|Default|SetShotsHitPlayer 0)
          (Variables|Default|SetIsPlayerDead false)
          (Variables|Default|SetPlayerHealth 200)
          (Variables|Default|SetshouldTimerStartAtZero true))))))

(fn ClearTextArray
  (params (TextArray :type Array%20of%20Text%20Render%20Component%20Object%20References))
  (if (> (Utilities|Array|Length TextArray) 0)
    (Components|DestroyComponent (Utilities|Array|Get(acopy) TextArray))
    (Utilities|Array|Clear TextArray)))

(fn ConstructionScript
  (Utilities|Array|Add (Variables|Default|GetPlayerCollisionArray) (Transformation|GetWorldLocation (Variables|Default|GetHeadCollision)))
  (Utilities|Array|Add (Variables|Default|GetToleranceLevelsForPCA) 13.0)
  (Utilities|Array|Add (Variables|Default|GetPlayerCollisionArray) (Transformation|GetWorldLocation (Variables|Default|GetTorsoCollision)))
  (Utilities|Array|Add (Variables|Default|GetToleranceLevelsForPCA) 32.0)
  (Utilities|Array|Add (Variables|Default|GetPlayerCollisionArray) (Transformation|GetWorldLocation (Variables|Default|GetLeftArmCollision)))
  (Utilities|Array|Add (Variables|Default|GetToleranceLevelsForPCA) 1e+01)
  (Utilities|Array|Add (Variables|Default|GetPlayerCollisionArray) (Transformation|GetWorldLocation (Variables|Default|GetRightArmCollision)))
  (Utilities|Array|Add (Variables|Default|GetToleranceLevelsForPCA) 1e+01)
  (Utilities|Array|Add (Variables|Default|GetPlayerCollisionArray) (Transformation|GetWorldLocation (Variables|Default|GetLeftHandCollision)))
  (Utilities|Array|Add (Variables|Default|GetToleranceLevelsForPCA) 7.0)
  (Utilities|Array|Add (Variables|Default|GetPlayerCollisionArray) (Transformation|GetWorldLocation (Variables|Default|GetRightHandCollision)))
  (Utilities|Array|Add (Variables|Default|GetToleranceLevelsForPCA) 7.0)
  (Utilities|Array|Add (Variables|Default|GetPlayerCollisionArray) (Transformation|GetWorldLocation (Variables|Default|GetLeftLegCollision)))
  (Utilities|Array|Add (Variables|Default|GetToleranceLevelsForPCA) 18.0)
  (Utilities|Array|Add (Variables|Default|GetPlayerCollisionArray) (Transformation|GetWorldLocation (Variables|Default|GetRightLegCollision)))
  (Utilities|Array|Add (Variables|Default|GetToleranceLevelsForPCA) 18.0)
  (Utilities|Array|Add (Variables|Default|GetPlayerCollisionArray) (Transformation|GetWorldLocation (Variables|Default|GetLeftFootCollision)))
  (Utilities|Array|Add (Variables|Default|GetToleranceLevelsForPCA) 1e+01)
  (Utilities|Array|Add (Variables|Default|GetPlayerCollisionArray) (Transformation|GetWorldLocation (Variables|Default|GetRightFootCollision)))
  (Utilities|Array|Add (Variables|Default|GetToleranceLevelsForPCA) 1e+01)
  (|CallParentFunction))
