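; Wood mine archway Blueprint actor, a wall building piece that can play a wood creak sound.
; On BeginPlay it calls the parent implementation, then plays its wood creak audio component if the creak sound flag is set.
; The construction script only runs the parent's logic.
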(event EventBeginPlay
  (|CallParentFunction)
  (if (Variables|Default|GetPlayCreakSound)
    (Audio|Components|Audio|Play (Variables|WoodMines_Archway|GetWoodCreakSound) 0.0)))

(fn ConstructionScript
  (|CallParentFunction))
